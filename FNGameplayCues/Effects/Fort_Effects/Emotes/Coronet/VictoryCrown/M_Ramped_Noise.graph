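# Additive unlit Niagara sprite and mesh particle material for the VictoryCrown emote effects, drawing a noise-distorted, curve-shaped glow.
# Panning smoke noise, randomized and scaled by dynamic parameters, offsets UVs, weighted by the inverse of a U-ramp curve.
# Curve-atlas U and V ramps at those warped UVs form a mask that indexes a color curve, tinted by particle color.
# Particle alpha, near-camera fade and depth fade shape opacity.
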Material /FNGameplayCues/Effects/Fort_Effects/Emotes/Coronet/VictoryCrown/M_Ramped_Noise
graph {
 properties {
  bUsedWithNiagaraMeshParticles=true
  bUsedWithNiagaraSprites=true
  blendMode=BLEND_Additive
  pixelDepthOffsetMode=PDOM_Legacy
  shadingModel=MSM_Unlit
  usedShadingModels=Unlit
 }
 nodes [
  n1 = CurveAtlasRowParameter(CurveTime=MaterialFunctionCall(Float2=TextureCoordinate()) {materialFunction=&BreakOutFloat2Components}.R) {
   atlas=/Game/Effects/Fort_Effects/Emotes/Coronet/VictoryCrown/Curves/CA_VC_curves
   curve=/Game/Effects/Fort_Effects/Emotes/Coronet/VictoryCrown/Curves/C_VC_uRamp_D
   defaultValue=6
   parameterName=Opacity_U
  }
  n2 = OneMinus(None=@n1.R)
  n3 = DynamicParameter() {parameterIndex=3, defaultValue=<rgba 0 0 1 1>, paramNames [UVTilingRandom, TimeOffsetRandom, SplineWidthBase, SplineWidthTip]}
  n4 = Add(
   A=LinearInterpolate(
    A=TextureCoordinate()
    B=TextureCoordinate() {coordinateIndex=1}
    Alpha=ScalarParameter() {group="Noise ", parameterName="Use Noise TexCoord [0] or [1]"}
   )
   B=@n3.UVTilingRandom
  )
  n5 = AppendVector(
   A=ScalarParameter() {defaultValue=1, parameterName=NoiseScaleU, sliderMax=1}
   B=ScalarParameter() {defaultValue=0.3, parameterName=NoiseScaleV, sliderMax=1}
  )
  n6 = Time()
  n7 = ScalarParameter() {parameterName=PanSpeedU, sliderMax=1}
  n8 = ScalarParameter() {defaultValue=0.1, parameterName=PanSpeedV, sliderMax=1}
  n9 = TextureSample(UVs=Panner(Coordinate=Multiply(A=@n4, B=@n5), Time=Add(A=@n6, B=@n3.TimeOffsetRandom), Speed=AppendVector(A=@n7, B=@n8)) {speedX=0.1}) {texture=&T_Inky_Smoke_Tile}
  n10 = ScalarParameter() {defaultValue=-0.5, group="Noise ", parameterName=Noise Strength}
  n11 = DynamicParameter() {defaultValue=<rgba 0 1 1 1>, paramNames [Emi V Offset, Emi V Tiling, Noise, Dissolve]}
  n12 = Add(
   A=Multiply(A=ComponentMask(None=@n9.RGB) {g=true, r=true}, B=Multiply(A=@n10, B=@n11.Noise))
   B=ScalarParameter() {group="Noise ", parameterName=Noise Offset}
  )
  n13 = MaterialFunctionCall(Float2=Add(A=TextureCoordinate(), B=Multiply(A=@n2, B=@n12))) {materialFunction=&BreakOutFloat2Components}
  n14 = CurveAtlasRowParameter(CurveTime=@n13.R) {
   atlas=&CA_VC_curves
   curve=&C_VC_uRamp_D
   defaultValue=6
   parameterName=Opacity_U
  }
  n15 = CurveAtlasRowParameter(CurveTime=@n13.G) {
   atlas=&CA_VC_curves
   curve=/Game/Effects/Fort_Effects/Emotes/Coronet/VictoryCrown/Curves/C_VC_vRamp_F
   defaultValue=7
   parameterName=Opacity_V
  }
  n16 = Multiply(A=@n14, B=@n15)
  n17 = ScalarParameter() {defaultValue=1, parameterName=ColorWeight}
  n18 = CurveAtlasRowParameter(CurveTime=MaterialFunctionCall(Float3=Multiply(A=@n16, B=@n17)) {materialFunction=&BreakOutFloat3Components}.R) {
   atlas=&CA_VC_curves
   curve=/Game/Effects/Fort_Effects/Emotes/Coronet/VictoryCrown/Curves/CA_VC_colCurve_A
   defaultValue=2
  }
  n19 = ParticleColor()
  MaterialFunctionCall(Emissive=Multiply(A=@n18, B=ComponentMask(None=@n19.RGB) {b=true, g=true, r=true})) {materialFunction=&MF_FX_EmissiveAdjustments}
  n20 = ScalarParameter() {defaultValue=125, parameterName=NearCamFade_Distance}
  n21 = ScalarParameter() {defaultValue=15, parameterName=DepthFade_Distance, sliderMax=125}
  Multiply(
   A=Multiply(
    A=MaterialFunctionCall(Distance=@n20) {materialFunction=&Near_Camera_Fade}.Result
    B=Multiply(A=@n16, B=@n19.A)
   )
   B=Power(Base=DepthFade(FadeDistance=@n21), Exp=ScalarParameter() {defaultValue=1, parameterName=DepthFade_Exponent, sliderMax=15})
  )
 ]
}
